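# Unlit additive Niagara sprite material for gravity distortion FX on the male Jupiter pickaxe, pairing refraction with a cellular glow.
# A radial gradient, depth fade and inverted camera fade mask an index-of-refraction lerp, while a panned, offset normal map supplies distortion detail.
# Inverted, sharpened cellular noise, masked and tinted by particle color, drives emissive.
# Dynamic parameters control the effect, though U_PAN_SPEED feeds both pan axes and CameraFadeLength feeds both camera fade inputs.
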Material /BRCosmetics/Weapons/FORT_Melee/Pickaxe_Jupiter_Male/FX/M_GravityDistortion
graph {
 properties {
  bUsedWithNiagaraSprites=true
  blendMode=BLEND_Additive
  pixelDepthOffsetMode=PDOM_Legacy
  refractionMethod=RM_IndexOfRefraction
  shadingModel=MSM_Unlit
  usedShadingModels=Unlit
 }
 nodes [
  n1 = DynamicParameter() {desc=Dynamic Parameter Index 0, defaultValue=<rgba 0 0 1.2 50>, paramNames [U_OFFSET, V_OFFSET, DISTORTION_AMT, DEPTH_FADE]}
  n2 = ParticleColor()
  n3 = DepthFade(
   Opacity=@n2.A
   FadeDistance=DynamicParameter() {defaultValue=<rgba 0 0 1.2 50>, paramNames [U_OFFSET, V_OFFSET, DISTORTION_AMT, DEPTH_FADE]}.DEPTH_FADE
  )
  n4 = MaterialFunctionCall() {materialFunction=&RadialGradientExponential}
  n5 = DynamicParameter() {
   desc=Dynamic Parameter Index 1
   parameterIndex=1
   defaultValue=<rgba 0 0 0 0>
   paramNames [U_PAN_SPEED, V_PAN_SPEED, CameraFadeLength, FadeOffset]
  }
  n6 = MaterialFunctionCall(Fade Length=@n5.CameraFadeLength, Fade Offset=@n5.CameraFadeLength) {materialFunction=&CameraDepthFade}
  LinearInterpolate(B=@n1.DISTORTION_AMT, Alpha=Multiply(A=Multiply(A=@n3, B=@n4.RadialGradientExponential), B=OneMinus(None=@n6.Result))) {constA=1, constB=1.05}
  n7 = TextureSample() {samplerType=SAMPLERTYPE_Masks, texture=/Game/Global/Textures/Default/Noise/T_CellularNoise}
  n8 = Multiply(A=Power(Base=OneMinus(None=@n7.G)) {constExponent=5}, B=@n4.RadialGradientExponential)
  MaterialFunctionCall(Emissive=Multiply(A=Multiply(A=@n8, B=@n3), B=@n2.RGB)) {materialFunction=&MF_FX_EmissiveAdjustments}
  n9 = MaterialFunctionCall(Float2=TextureCoordinate()) {materialFunction=&BreakOutFloat2Components}
  n10 = Panner(
   Coordinate=AppendVector(A=Add(A=@n9.R, B=@n1.U_OFFSET), B=Add(A=@n9.G, B=@n1.V_OFFSET))
   Speed=AppendVector(A=@n5.U_PAN_SPEED, B=@n5.U_PAN_SPEED)
  )
  Normalize(VectorInput=LinearInterpolate(
    A=Constant3Vector() {constant=<rgba 0 0 1 0>}
    B=TextureSampleParameter2D(UVs=@n10) {parameterName=Normal Map, samplerType=SAMPLERTYPE_Normal, texture=/Game/Global/Textures/Default/Tileable_Patterns/1Normal}.RGB
    Alpha=ScalarParameter() {defaultValue=0.3, parameterName=Alpha}
   ))
 ]
}
