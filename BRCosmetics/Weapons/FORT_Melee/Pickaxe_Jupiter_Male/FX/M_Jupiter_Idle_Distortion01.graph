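# Translucent unlit Niagara sprite/ribbon material for the idle distortion effect on the Jupiter male pickaxe, using index-of-refraction distortion.
# A powered sine of the U gradient and a Spawngroup offset warp panning UVs for blurred noise and normal samples.
# The noise, masked by a sharpened V-gradient sine, drives depth-faded opacity, refraction blend and particle-colored emissive.
# Dynamic parameters control tiling, speed and refraction; U_Speed feeds both pan axes, leaving V_Speed unused.
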Material /BRCosmetics/Weapons/FORT_Melee/Pickaxe_Jupiter_Male/FX/M_Jupiter_Idle_Distortion01
graph {
 properties {
  bUsedWithNiagaraRibbons=true
  bUsedWithNiagaraSprites=true
  blendMode=BLEND_Translucent
  pixelDepthOffsetMode=PDOM_Legacy
  refractionMethod=RM_IndexOfRefraction
  shadingModel=MSM_Unlit
  usedShadingModels=Unlit
 }
 nodes [
  n1 = DynamicParameter() {parameterIndex=1, defaultValue=<rgba 1 1 0 0>, paramNames [U_Scale,V_Scale,U_Speed,V_Speed]}
  n2 = DynamicParameter() {defaultValue=<rgba 0 0 0 0>, paramNames [Refraction_LERP, Spawngroup, DepthFade, Refraction_Level]}
  n3 = OneMinus(None=Divide(B=Add(A=@n2.Spawngroup)))
  n4 = MaterialFunctionCall() {materialFunction=&LinearGradient}
  n5 = Power(Base=Sine(None=Multiply(A=@n4.UGradient) {constB=0.5}), Exp=ScalarParameter() {defaultValue=5, parameterName=CenterOut Moverment PWR})
  n6 = MaterialFunctionCall(Float2=TextureCoordinate()) {materialFunction=&BreakOutFloat2Components}
  n7 = MaterialFunctionCall(Float2=AppendVector(A=Add(A=@n3, B=@n5), B=Add(A=@n3, B=@n6.G))) {materialFunction=&BreakOutFloat2Components}
  n8 = Panner(
   Coordinate=AppendVector(A=Multiply(A=@n1.U_Scale, B=@n7.R), B=Multiply(A=@n7.G, B=@n1.V_Scale))
   Speed=AppendVector(A=@n1.U_Speed, B=@n1.U_Speed)
  ) {speedX=1}
  n9 = Sine(None=Multiply(A=MaterialFunctionCall() {materialFunction=&LinearGradient}.VGradient) {constB=0.5})
  n10 = Multiply(A=@n9, B=@n9)
  n11 = Multiply(
   A=Power(
    Base=TextureSample(UVs=@n8) {samplerType=SAMPLERTYPE_Grayscale, texture=&LowResBlurredNoise}.G
    Exp=ScalarParameter() {defaultValue=1.25, parameterName=Refraction Noise Power}
   )
   B=Multiply(A=@n10, B=@n10)
  )
  n12 = DynamicParameter() {defaultValue=<rgba 0 0 0 0>, paramNames [Refraction_LERP, Spawngroup, DepthFade, Refraction_Level]}
  n13 = Multiply(A=@n11, B=@n12.Refraction_LERP)
  DepthFade(
   Opacity=@n13
   FadeDistance=DynamicParameter() {defaultValue=<rgba 0 0 0 0>, paramNames [Refraction_LERP, Spawngroup, DepthFade, Refraction_Level]}.DepthFade
  )
  DynamicParameter() {defaultValue=<rgba 0 0 0 0>, paramNames [Refraction_LERP, Spawngroup, DepthFade, Refraction_Level]}
  LinearInterpolate(B=@n12.Refraction_Level, Alpha=@n13) {constA=1, constB=1.1}
  n14 = ParticleColor()
  MaterialFunctionCall(Emissive=Multiply(A=Multiply(A=@n11, B=@n14.RGB), B=Power(Base=@n13) {constExponent=0.5})) {materialFunction=&MF_FX_EmissiveAdjustments}
  TextureSample(UVs=@n8) {samplerType=SAMPLERTYPE_Normal, texture=/Game/Global/Textures/Default/Noise/T_Normal_Noise01}
 ]
}
