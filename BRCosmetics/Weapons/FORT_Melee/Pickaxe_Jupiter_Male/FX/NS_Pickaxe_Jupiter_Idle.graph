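# Niagara idle effect for the Jupiter pickaxe, adding blue energy ribbons, electricity, glows and a refraction aura to skeletal mesh sockets.
# Two layered distortion ribbon emitters shape four socket-following groups each, using vector curves over custom link order.
# Electricity ribbons burst randomly at spawn-sampled sockets, while Lights adds persistent sprites and lights.
# Refraction_Aura spawns continuous socket-tracking sprites, with brightness, camera offset and material values adjusted by the front-end preview flag.
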NiagaraSystem /BRCosmetics/Weapons/FORT_Melee/Pickaxe_Jupiter_Male/FX/NS_Pickaxe_Jupiter_Idle
dependencies {
 usedDataInterfaces [&NiagaraDataInterfaceVectorCurve, &NiagaraDataInterfaceCurve]
 usedDynamicInputs []
 usedModules [&SystemState]
}
emitters {
 Distortion_BLUE_C01 {
  emitterSpawnScript [/SetVariables(Emitter.NumSpawgroups=4, Emitter.ParticlesPerSpawngroup=12) {isSetParameters=true, module=SetVariables_#1}]
  emitterUpdateScript [
   &EmitterState(
    Inactive Response=::NewEnumerator1
    Life Cycle Mode=::NewEnumerator1
    Loop Behavior=::NewEnumerator1
    Loop Duration Mode=::NewEnumerator1
    UseLoopDelay=false
   )
   &SpawnBurst_Instantaneous(
    Spawn Count=&Multiply_Int(A=$Emitter.NumSpawgroups, B=$Emitter.ParticlesPerSpawngroup)
    Use Spawn Probability=false
   )
  ]
  particleSpawnScript [
   /SetVariables(Particles.Spawngroup:NiagaraInt32=&self:CustomSpawngroups(Number of Spawngroups:NiagaraInt32=$Emitter.NumSpawgroups)) {isSetParameters=true, module=SetVariables_#2}
   /SetVariables(Particles.RibbonLinkOrder:NiagaraFloat=&self:CustomRibbonLinkOrder()) {isSetParameters=true, module=SetVariables_#3}
   /SetVariables(Particles.RibbonEndsGradient:NiagaraFloat=&self:ScratchDynamicInput_01()) {isSetParameters=true, module=SetVariables_#4}
   &InitializeRibbon(
    Color=&MultiplyLinearColor(
     Linear Color=<rgba 0 8.8889 20 1>
     Scale Factor=&MakeVector4_FromFloat(Float=&Lerp_Float(A=2, B=-0.5, Alpha=$Particles.RibbonEndsGradient))
    )
    Module.Write Color=true
    Module.Write Lifetime=true
    Module.Write Mass=true
    Module.Write Material Random=true
    Module.Write Position=false
    Module.Write RibbonTwist=false
    Module.Write RibbonWidth=true
    Ribbon Width=10
   )
   /SetVariables() {isSetParameters=true, module=SetVariables_#5}
   &DynamicMaterialParameters(
    Index 0 Param 1=&Power(a=$Particles.RibbonEndsGradient, b=1.125)
    Index 0 Param 2=&MakeFloatFromInt(Int=$Particles.Spawngroup)
    Index 0 Param 4=1.2
    Index 1 Param 1=0.65
    Index 1 Param 2=1
    Index 1 Param 3=-0.666
    Module.Param0WriteEnabled=true
    Module.Param0WriteEnabled1=true
    Module.Param1WriteEnabled=true
    Module.Param1WriteEnabled1=true
    Module.Param2WriteEnabled=true
    Module.Param2WriteEnabled1=true
    Module.Param3WriteEnabled=true
    Module.Param3WriteEnabled1=true
    Write Parameter Index 1=true
   )
  ]
  particleUpdateScript [
   &SkeletalMeshLocation(
    Bone / Socket Index=&Add_Integer(A=3, B=$Particles.Spawngroup)
    Inherit Velocity=false
    MaskSpawnGroup=false
    Mesh Sampling Type=::NewEnumerator5
    Sampled Position Offset=&Add_Vector(
     A=<0 0.25 0>
     B=&RotateVectorByQuaternion(
      Vector To Rotate=&VectorFromCurve(
       VectorCurve=DataInterface() {propertyValues {
         XCurve {keys [
           {arriveTangent=854.0034, interpMode=RCIM_Cubic, leaveTangent=75.099, tangentMode=RCTM_User, time=0, value=-16}
           {arriveTangent=28.5485, leaveTangent=6.5462, time=0.2014, value=-0.0697}
           <at 0.4975 -0.8647>
           {arriveTangent=-21.0593, interpMode=RCIM_Cubic, leaveTangent=-21.0593, tangentMode=RCTM_User, time=0.8014, value=-0.0697}
           {arriveTangent=-63.3649, interpMode=RCIM_Cubic, leaveTangent=-37284.9844, tangentMode=RCTM_User, time=1, value=-16}
          ]}
         YCurve {keys [<at 0 -11>,<at 1 10>]}
         ZCurve {keys [<at 0 0>,<at 0.2 0>,<at 0.8 4>,<at 1.0033 4>]}
        }}
       CurveIndex=$Particles.RibbonLinkOrder
       Scale Curve=&MakeVector(X=1, Y=1, Z=&OneMinusFloat(Float=$Particles.RibbonEndsGradient))
      )
      Quaternion=&MakeQuatFromAxisAngle(Axis=<0 0 1>, AngleInDegrees=270)
     )
    )
    Sampled Position Offset Coordinate Space:ENiagaraMeshTransforms=::NewEnumerator3
    Skeletal Mesh=$User.SkeletalMesh
    Socket Sampling Mode=::NewEnumerator5
   )
   &ParticleState(Kill Particles When Lifetime Has Elapsed=false, Loop Particles Lifetime=true)
  ]
  rendererClasses [&NiagaraRibbonRendererProperties]
  renderers [{rendererClass=&NiagaraRibbonRendererProperties}]
 }
 Distortion_BLUE_C02 {
  emitterSpawnScript [/SetVariables(Emitter.NumSpawgroups=4, Emitter.ParticlesPerSpawngroup=12) {isSetParameters=true, module=SetVariables_#1}]
  emitterUpdateScript [
   &EmitterState(
    Inactive Response=::NewEnumerator1
    Life Cycle Mode=::NewEnumerator1
    Loop Behavior=::NewEnumerator1
    Loop Duration Mode=::NewEnumerator1
    UseLoopDelay=false
   )
   &SpawnBurst_Instantaneous(
    Spawn Count=&Multiply_Int(A=$Emitter.NumSpawgroups, B=$Emitter.ParticlesPerSpawngroup)
    Use Spawn Probability=false
   )
  ]
  particleSpawnScript [
   /SetVariables(Particles.Spawngroup:NiagaraInt32=&self:CustomSpawngroups(Number of Spawngroups:NiagaraInt32=$Emitter.NumSpawgroups)) {isSetParameters=true, module=SetVariables_#2}
   /SetVariables(Particles.RibbonLinkOrder:NiagaraFloat=&self:CustomRibbonLinkOrder()) {isSetParameters=true, module=SetVariables_#3}
   /SetVariables(Particles.RibbonEndsGradient:NiagaraFloat=&self:ScratchDynamicInput_01()) {isSetParameters=true, module=SetVariables_#4}
   &InitializeRibbon(
    Color=&MultiplyLinearColor(
     Linear Color=<rgba 5 45 100 1>
     Scale Factor=&MakeVector4_FromFloat(Float=&Lerp_Float(A=0, B=1, Alpha=$Particles.RibbonEndsGradient))
    )
    Module.Write Color=true
    Module.Write Lifetime=true
    Module.Write Mass=true
    Module.Write Material Random=true
    Module.Write Position=false
    Module.Write RibbonTwist=false
    Module.Write RibbonWidth=true
    Ribbon Width=18
   )
   /SetVariables() {isSetParameters=true, module=SetVariables_#5}
   &DynamicMaterialParameters(
    Index 0 Param 1=&Power(a=$Particles.RibbonEndsGradient, b=6)
    Index 0 Param 2=&MakeFloatFromInt(Int=$Particles.Spawngroup)
    Index 0 Param 4=1
    Index 1 Param 1=0.5
    Index 1 Param 2=1.25
    Module.Param0WriteEnabled=true
    Module.Param0WriteEnabled1=true
    Module.Param1WriteEnabled=true
    Module.Param1WriteEnabled1=true
    Module.Param2WriteEnabled=true
    Module.Param2WriteEnabled1=true
    Module.Param3WriteEnabled=true
    Module.Param3WriteEnabled1=true
    Write Parameter Index 1=true
   )
  ]
  particleUpdateScript [
   &SkeletalMeshLocation(
    Bone / Socket Index=&Add_Integer(A=3, B=$Particles.Spawngroup)
    Inherit Velocity=false
    MaskSpawnGroup=false
    Mesh Sampling Type=::NewEnumerator5
    Sampled Position Offset=&Add_Vector(
     A=<0 0.25 0>
     B=&RotateVectorByQuaternion(
      Vector To Rotate=&VectorFromCurve(
       VectorCurve=DataInterface() {propertyValues {
         XCurve {keys [
           {arriveTangent=854.0034, interpMode=RCIM_Cubic, leaveTangent=75.099, tangentMode=RCTM_User, time=0, value=-16}
           {arriveTangent=28.5485, leaveTangent=6.5462, time=0.2014, value=-0.0697}
           <at 0.4975 -0.8647>
           {arriveTangent=-21.0593, interpMode=RCIM_Cubic, leaveTangent=-21.0593, tangentMode=RCTM_User, time=0.8014, value=-0.0697}
           {arriveTangent=-63.3649, interpMode=RCIM_Cubic, leaveTangent=-37284.9844, tangentMode=RCTM_User, time=1, value=-16}
          ]}
         YCurve {keys [<at 0 -11>,<at 1 10>]}
         ZCurve {keys [<at 0 0>,<at 0.2 0>,<at 0.8 4>,<at 1.0033 4>]}
        }}
       CurveIndex=$Particles.RibbonLinkOrder
       Scale Curve=&MakeVector(X=1, Y=1, Z=&OneMinusFloat(Float=$Particles.RibbonEndsGradient))
      )
      Quaternion=&MakeQuatFromAxisAngle(Axis=<0 0 1>, AngleInDegrees=270)
     )
    )
    Sampled Position Offset Coordinate Space:ENiagaraMeshTransforms=::NewEnumerator3
    Skeletal Mesh=$User.SkeletalMesh
    Socket Sampling Mode=::NewEnumerator5
   )
   &ParticleState(Kill Particles When Lifetime Has Elapsed=false, Loop Particles Lifetime=true)
  ]
  rendererClasses [&NiagaraRibbonRendererProperties]
  renderers [{rendererClass=&NiagaraRibbonRendererProperties}]
 }
 Electricity {
  emitterSpawnScript [/SetVariables(Emitter.NumSpawgroups=1, Emitter.ParticlesPerSpawngroup=4) {isSetParameters=true, module=SetVariables_#1}]
  emitterUpdateScript [
   &EmitterState(
    Life Cycle Mode=::NewEnumerator1
    Loop Duration=&RandomRangeFloat(Minimum=0.2, Maximum=0.2)
    Recalculate Duration Each Loop=true
    UseLoopDelay=false
   )
   &SpawnBurst_Instantaneous(
    Spawn Count=&Multiply_Int(A=$Emitter.NumSpawgroups, B=$Emitter.ParticlesPerSpawngroup)
    Spawn Probability=0.5
    Use Spawn Probability=true
   )
  ]
  particleSpawnScript [
   /SetVariables(Particles.Spawngroup:NiagaraInt32=&self:CustomSpawngroups(Number of Spawngroups:NiagaraInt32=$Emitter.NumSpawgroups)) {isSetParameters=true, module=SetVariables_#2}
   /SetVariables(Particles.RibbonLinkOrder:NiagaraFloat=&self:CustomRibbonLinkOrder()) {isSetParameters=true, module=SetVariables_#3}
   /SetVariables(Particles.RibbonEndsGradient:NiagaraFloat=&self:ScratchDynamicInput_01()) {isSetParameters=true, module=SetVariables_#4}
   &InitializeRibbon(
    Color=<rgba 0.3 1 20 1>
    Lifetime=&RandomRangeFloat(Minimum=0.265, Maximum=0.666)
    Module.Write Color=true
    Module.Write Lifetime=true
    Module.Write Mass=false
    Module.Write Material Random=true
    Module.Write Position=false
    Module.Write RibbonTwist=false
    Module.Write RibbonWidth=true
    Ribbon Width=&RandomRangeFloat(Minimum=6, Maximum=10)
   )
   &SkeletalMeshLocation(
    Bone / Socket Index=&Add_Integer(A=$Particles.Spawngroup, B=&RandomRangeInt(Minimum=0, Maximum=2))
    Inherit Velocity=false
    MaskSpawnGroup=false
    Mesh Sampling Type=::NewEnumerator5
    Sampled Position Offset=&Add_Vector(
     A=<0 0 0>
     B=&RotateVectorByQuaternion(
      Vector To Rotate=&VectorFromCurve(
       VectorCurve=DataInterface() {propertyValues {
         XCurve {keys [
           {arriveTangent=854.0034, interpMode=RCIM_Cubic, leaveTangent=75.099, tangentMode=RCTM_User, time=0, value=-4}
           {arriveTangent=28.5485, leaveTangent=6.5462, time=0.2014, value=-0.0697}
           <at 0.5003 -1.7271>
           {arriveTangent=-21.0593, interpMode=RCIM_Cubic, leaveTangent=-21.0593, tangentMode=RCTM_User, time=0.8014, value=-0.0697}
           {arriveTangent=-63.3649, interpMode=RCIM_Cubic, leaveTangent=-37284.9844, tangentMode=RCTM_User, time=1, value=-4}
          ]}
         YCurve {keys [<at 0 -1>,<at 1 1>]}
         ZCurve {keys [<at 0 0>,<at 0.2 0>,<at 0.8 4>,<at 1.0033 4>]}
        }}
       CurveIndex=$Particles.RibbonLinkOrder
       Scale Curve=&MakeVector(X=1, Y=1, Z=&OneMinusFloat(Float=$Particles.RibbonEndsGradient))
      )
      Quaternion=&MakeQuatFromAxisAngle(Axis=<0 0 1>, AngleInDegrees=270)
     )
    )
    Sampled Position Offset Coordinate Space:ENiagaraMeshTransforms=::NewEnumerator3
    Skeletal Mesh=$User.SkeletalMesh
    Socket Sampling Mode=::NewEnumerator5
   )
   /SetVariables() {isSetParameters=true, module=SetVariables_#5}
   /SetVariables(Particles.LightEnabled:NiagaraBool=&SetBoolByFloatComparison(A=$Particles.MaterialRandom, B=0.35, Comparison Type=::NewEnumerator6)) {isSetParameters=true}
   &CameraOffset(Camera Offset Amount=3)
   &DynamicMaterialParameters(
    Index 0 Param 1=&RandomRangeFloat(Minimum=-0.333, Maximum=0.333)
    Module.Param0WriteEnabled=true
    Module.Param1WriteEnabled=true
    Module.Param2WriteEnabled=false
    Module.Param3WriteEnabled=false
   )
   /SetVariables(Particles.LightColor=<rgba 0.3 0.3 20 1>) {isSetParameters=true}
  ]
  particleUpdateScript [
   &ParticleState()
   &ScaleColor(
    Scale Alpha=/Niagara/DynamicInputs/Math/LerpMultipleFloats(
     Float Count:ENiagara_MultipleLerpCount=::NewEnumerator0
     1:NiagaraFloat=0
     2:NiagaraFloat=2
     3:NiagaraFloat=0
     4:NiagaraFloat=?
     5:NiagaraFloat=?
     Alpha:NiagaraFloat=$Particles.Initial.RibbonLinkOrder
    )
    ScaleA=true
    ScaleRGB=true
   )
  ]
  rendererClasses [&NiagaraRibbonRendererProperties]
  renderers [{rendererClass=&NiagaraRibbonRendererProperties}]
 }
 Lights {
  emitterUpdateScript [
   &EmitterState(
    Inactive Response=::NewEnumerator1
    Life Cycle Mode=::NewEnumerator1
    Loop Behavior=::NewEnumerator1
    Loop Duration Mode=::NewEnumerator1
    UseLoopDelay=false
   )
   &SpawnBurst_Instantaneous(Spawn Count=4, Use Spawn Probability=false)
  ]
  particleSpawnScript [
   /SetVariables(Particles.Spawngroup:NiagaraInt32=&Add_Integer(A=3, B=&ModuloInt(A=&ReturnExecIndex(), B=4))) {isSetParameters=true}
   /Niagara/Modules/Spawn/Initialization/InitializeParticle(
    Color=&MultiplyLinearColor(Linear Color=<rgba 0.4 0.9333 2 0.3>, Scale Factor=&MakeVector4_FromFloat(Float=1))
    Lifetime=5
    Mass=3
    Module.Write Color=true
    Module.Write Lifetime=true
    Module.Write Mass=true
    Module.Write Material Random=true
    Module.Write Position=false
    Module.Write Scale=false
    Module.Write SpriteRotation=true
    Module.Write SpriteSize=true
    Position=&SimulationPosition()
    Sprite Rotation=-15
    Sprite Size=&Vector2DFromFloat(Value=10)
   )
   &CameraOffset(Camera Offset Amount=6.25)
   &DynamicMaterialParameters(
    Index 0 Param 1=0.8
    Index 0 Param 2=1
    Index 0 Param 3=0.333
    Index 1 Param 1=60
    Index 1 Param 2=0.2
    Index 1 Param 3=1
    Index 1 Param 4=2
    Module.Param0WriteEnabled=true
    Module.Param0WriteEnabled1=true
    Module.Param1WriteEnabled=true
    Module.Param1WriteEnabled1=true
    Module.Param2WriteEnabled=true
    Module.Param2WriteEnabled1=true
    Module.Param3WriteEnabled=true
    Module.Param3WriteEnabled1=true
    Write Parameter Index 1=true
   )
   /SetVariables(
    Particles.ColorSkew:LinearColor=&MultiplyLinearColor(Linear Color=<rgba 0.0225 0.0347 0.05 0.05>, Scale Factor=&MakeVector4_FromFloat(Float=1.5))
    Particles.PositionOffset:Vector3f=<-0.5 0 0>
    Particles.SpriteScaleSkew:Vector2f=&Vector2DFromFloat(Value=80)
    Particles.UVScaleSkew:Vector2f=<1 6>
   ) {isSetParameters=true}
  ]
  particleUpdateScript [
   &ParticleState(Kill Particles When Lifetime Has Elapsed=false)
   &SkeletalMeshLocation(
    Bone / Socket Index=$Particles.Initial.Spawngroup
    Inherit Velocity=false
    MaskSpawnGroup=false
    Mesh Sampling Type=::NewEnumerator5
    Sampled Position Offset=<1 0 0>
    Sampled Position Offset Coordinate Space:ENiagaraMeshTransforms=::NewEnumerator2
    Skeletal Mesh=$User.SkeletalMesh
    Socket Sampling Mode=::NewEnumerator5
   )
   /SetVariables(
    Particles.LightEnabled:NiagaraBool=&SetBoolByIntComparison(A=&ModuloInt(A=$Particles.Initial.Spawngroup, B=4), B=2, Comparison Type=::NewEnumerator4)
    Particles.PositionInset:Vector3f=&Add_Vector(
     A=$Output.SkeletalMeshLocation.SampledPosition
     B=&TransformVector(Vector=$Particles.Initial.PositionOffset, Destination Space=::Simulation)
    )
    Particles.SpriteAlignment=<0 0 1>
    Particles.SpriteFacingSkew:Vector3f=&TransformVector(Vector=<1 0 0>, Destination Space=::World)
   ) {isSetParameters=true}
  ]
  rendererClasses [&NiagaraSpriteRendererProperties, &NiagaraLightRendererProperties]
  renderers [
   {rendererClass=&NiagaraSpriteRendererProperties}
   {rendererClass=&NiagaraLightRendererProperties, rendererIndex=1}
   {rendererClass=&NiagaraSpriteRendererProperties, rendererIndex=2}
  ]
 }
 Refraction_Aura {
  emitterUpdateScript [
   &EmitterState(Life Cycle Mode=::NewEnumerator1, Loop Duration=1, UseLoopDelay=false)
   &SpawnRate(SpawnRate=5, Use Spawn Probability=false)
  ]
  particleSpawnScript [
   /SetVariables(Particles.Spawngroup:NiagaraInt32=&Add_Integer(A=4, B=&ModuloInt(A=$Particles.UniqueID, B=3))) {isSetParameters=true, module=SetVariables_#2}
   &InitializeParticle(
    Color=&MakeLinearColorFromVectorAndFloat(
     "Vector (RGB)"=&Multiply_Vector(A=&VectorFromFloat(Value=1), B=<0 0.4 1.5>)
     "Float (Alpha)"=0.7
    )
    Lifetime=&RandomRangeFloat(Minimum=1.5, Maximum=1.8)
    Module.Write Color=true
    Module.Write Lifetime=true
    Module.Write Mass=true
    Module.Write Material Random=true
    Module.Write Position=false
    Module.Write Scale=false
    Module.Write SpriteRotation=true
    Module.Write SpriteSize=true
    Position=&SimulationPosition()
    Sprite Rotation=&RandomRangeFloat(Minimum=0, Maximum=359)
    Sprite Size=&Vector2DFromFloat(Value=&Multiply_Float(A=2, B=&RandomRangeFloat(Minimum=12, Maximum=16)))
   )
   &DynamicMaterialParameters(
    Index 0 Param 1=&RandomRangeFloat(Minimum=0, Maximum=1)
    Index 0 Param 2=&RandomRangeFloat(Minimum=0, Maximum=1)
    Index 0 Param 4=&MakeCustomFloatFromBool(Bool=$User.bIsFrontEndPreview, True Float=50, False Float=12.5)
    Index 1 Param 1=&RandomRangeFloat(Minimum=-0.2, Maximum=0.2)
    Index 1 Param 2=&RandomRangeFloat(Minimum=-0.2, Maximum=0.2)
    Index 1 Param 3=800
    Module.Param0WriteEnabled=true
    Module.Param0WriteEnabled1=true
    Module.Param1WriteEnabled=true
    Module.Param1WriteEnabled1=true
    Module.Param2WriteEnabled=false
    Module.Param2WriteEnabled1=true
    Module.Param3WriteEnabled=true
    Module.Param3WriteEnabled1=true
    Write Parameter Index 1=true
   )
  ]
  particleUpdateScript [
   &ParticleState()
   &SkeletalMeshLocation(
    Bone / Socket Index=$Particles.Initial.Spawngroup
    Inherit Velocity=false
    MaskSpawnGroup=false
    Mesh Sampling Type=::NewEnumerator5
    Sampled Position Offset=&MakeVector(X=&RandomRangeFloat(Minimum=-7, Maximum=-4), Y=0, Z=3)
    Sampled Position Offset Coordinate Space:ENiagaraMeshTransforms=::NewEnumerator2
    Skeletal Mesh=$User.SkeletalMesh
    Socket Sampling Mode=::NewEnumerator5
   )
   &ScaleSpriteSize(Scale Factor=&Vector2DFromFloat(Value=&FloatFromCurve(FloatCurve=DataInterface() {propertyValues {Curve {keys [{arriveTangent=6.3683, interpMode=RCIM_Cubic, time=0, value=0.8}, <at 1 1.25>]}}})))
   &ScaleColor(
    Scale RGB=&VectorFromFloat(Value=&FloatFromCurve(
      FloatCurve=DataInterface() {propertyValues {Curve {keys [
          {arriveTangent=-1.8749, interpMode=RCIM_Cubic, time=0, value=0}
          <at 0.2 1 RCIM_Cubic>
          {interpMode=RCIM_Cubic, leaveTangent=-3.25, time=1, value=0}
         ]}}}
      Scale Curve=/Niagara/DynamicInputs/Math/Floor_Float(Float:NiagaraFloat=&Add_Float(
        A=&MakeCustomFloatFromBool(Bool=$User.bIsFrontEndPreview, True Float=1.666, False Float=1)
        B=$Particles.Initial.MaterialRandom
       ))
     ))
    ScaleA=false
    ScaleRGB=true
   )
   &CameraOffset(Camera Offset Amount=&FloatFromCurve(
     FloatCurve=DataInterface() {propertyValues {Curve {keys [<at 0 0 RCIM_Cubic>,<at 1 22 RCIM_Cubic>]}}}
     Scale Curve=&MakeCustomFloatFromBool(Bool=$User.bIsFrontEndPreview, True Float=2, False Float=1)
    ))
   &DynamicMaterialParameters(
    Index 0 Param 3=&FloatFromCurve(FloatCurve=DataInterface() {propertyValues {Curve {keys [<at 0 1 RCIM_Cubic>,<at 0.5 1.085>,<at 1 1 RCIM_Cubic>]}}})
    Module.Param0WriteEnabled=false
    Module.Param1WriteEnabled=false
    Module.Param2WriteEnabled=true
    Module.Param3WriteEnabled=false
   ) {module=DynamicMaterialParameters001}
  ]
  rendererClasses [&NiagaraSpriteRendererProperties]
  renderers [{rendererClass=&NiagaraSpriteRendererProperties}]
 }
}
summary {
 systemName=NS_Pickaxe_Jupiter_Idle
 emitters [
  {emitterName=Electricity, rendererClasses [&NiagaraRibbonRendererProperties]}
  {emitterName=Refraction_Aura, rendererClasses [&NiagaraSpriteRendererProperties]}
  {emitterName=Lights, rendererClasses [&NiagaraSpriteRendererProperties, &NiagaraLightRendererProperties]}
  {emitterName=Distortion_BLUE_C01, rendererClasses [&NiagaraRibbonRendererProperties]}
  {emitterName=Distortion_BLUE_C02, rendererClasses [&NiagaraRibbonRendererProperties]}
 ]
 userVariables [
  User.bIsFrontEndPreview {type=NiagaraBool}
  User.Interpolate_Rate_of_Change {type=NiagaraFloat, defaultValue=0, description=""}
  User.Interpolate_WPO {type=NiagaraFloat, defaultValue=0, description=""}
  User.SkeletalMesh {
   type=NiagaraDataInterfaceSkeletalMesh
   defaultValue {
    struct=NiagaraExt_VariableValue_DataInterface
    value {dataInterface=&self:NiagaraDataInterfaceSkeletalMesh_0}
   }
  }
 ]
}
