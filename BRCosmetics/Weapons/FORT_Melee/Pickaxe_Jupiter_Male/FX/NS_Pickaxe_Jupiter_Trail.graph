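# Niagara ribbon trail for the Jupiter pickaxe, sampled from sockets on the User.SkeletalMesh.
# Two active rate-spawned emitters cycle particles across four socket indices using execution index modulo 4, producing short-lived, widening ribbons.
# LightTrail fades from bluish white through a color curve and sets a front-ribbon mask from loop age; MainTrail001 drives a dynamic material parameter over life.
# Burst-spawned versions of both emitters remain disabled.
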NiagaraSystem /BRCosmetics/Weapons/FORT_Melee/Pickaxe_Jupiter_Male/FX/NS_Pickaxe_Jupiter_Trail
dependencies {
 usedDataInterfaces [&NiagaraDataInterfaceColorCurve, &NiagaraDataInterfaceCurve]
 usedDynamicInputs []
 usedModules [&SystemState]
}
emitters {
 LightTrail {
  emitterUpdateScript [
   &EmitterState(Life Cycle Mode=::NewEnumerator1, Loop Behavior=::NewEnumerator1, Loop Duration=0.3, UseLoopDelay=false)
   &SpawnBurst_Instantaneous(Spawn Count=4, Use Spawn Probability=false) {enabled=false}
   &SpawnRate(SpawnRate=150, Use Spawn Probability=false)
  ]
  particleSpawnScript [
   /SetVariables(Particles.Spawngroup:NiagaraInt32=&Add_Integer(A=0, B=&ModuloInt(A=&ReturnExecIndex(), B=4))) {isSetParameters=true, module=SetVariables_#1}
   &InitializeRibbon(
    Color=<rgba 10 0 0 1>
    Lifetime=0.185
    Module.Write Color=false
    Module.Write Lifetime=true
    Module.Write Mass=false
    Module.Write Material Random=true
    Module.Write Position=false
    Module.Write RibbonTwist=false
    Module.Write RibbonWidth=true
    Ribbon Width=4
   )
   &SkeletalMeshLocation(
    Bone / Socket Index=$Particles.Initial.Spawngroup
    Inherit Velocity=false
    Interpolate Mesh Sampling=false
    MaskSpawnGroup=false
    Mesh Sampling Type=::NewEnumerator5
    Sampled Position Offset Coordinate Space:ENiagaraMeshTransforms=::NewEnumerator3
    Skeletal Mesh=$User.SkeletalMesh
    Socket Sampling Mode=::NewEnumerator5
   )
   /SetVariables(Particles.FrontRibbonMask:NiagaraFloat=&OneMinusFloat(Float=/Niagara/DynamicInputs/Integer/V2/Round_Float(Float:NiagaraFloat=$System.NormalizedLoopAge))) {isSetParameters=true, module=SetVariables_#2}
  ]
  particleUpdateScript [
   &ParticleState()
   &Color(Color=&ColorFromCurve(ColorCurve=DataInterface() {propertyValues {
       AlphaCurve {keys [<at 0 0.12 RCIM_Cubic>, {arriveTangent=-0.4613, interpMode=RCIM_Cubic, leaveTangent=-0.4613, tangentMode=RCTM_User, time=1, value=0}]}
       BlueCurve {keys [<at 0 1.333 RCIM_Constant>,<at 0.1874 0.8766>,<at 0.95 0>,<at 1 0>]}
       GreenCurve {keys [<at 0 1.3003 RCIM_Constant>,<at 0.1874 0.6366>,<at 0.85 0>,<at 1 0>]}
       RedCurve {keys [<at 0 1.2707 RCIM_Constant>,<at 0.1874 0.4166>,<at 0.75 0>,<at 1 0>]}
      }}))
   &ScaleRibbonWidth(Ribbon Width Scale=&FloatFromCurve(FloatCurve=DataInterface() {propertyValues {Curve {keys [<at 0 1 RCIM_Cubic>, {arriveTangent=-0.033, interpMode=RCIM_Cubic, leaveTangent=-0.033, tangentMode=RCTM_User, time=1, value=2}]}}}))
  ]
  rendererClasses [&NiagaraRibbonRendererProperties]
  renderers [{rendererClass=&NiagaraRibbonRendererProperties}]
 }
 LightTrail001 {
  bEnabled=false
  emitterUpdateScript [
   &EmitterState(Life Cycle Mode=::NewEnumerator1, Loop Duration=0.01, UseLoopDelay=false)
   &SpawnBurst_Instantaneous(Spawn Count=4, Use Spawn Probability=false)
  ]
  particleSpawnScript [
   /SetVariables(Particles.Spawngroup:NiagaraInt32=&Add_Integer(A=0, B=&ModuloInt(A=&ReturnExecIndex(), B=4))) {isSetParameters=true, module=SetVariables_#1}
   &InitializeRibbon(
    Color=<rgba 10 0 0 1>
    Lifetime=0.185
    Module.Write Color=false
    Module.Write Lifetime=true
    Module.Write Mass=false
    Module.Write Material Random=true
    Module.Write Position=false
    Module.Write RibbonTwist=false
    Module.Write RibbonWidth=true
    Ribbon Width=4
   )
   &SkeletalMeshLocation(
    Bone / Socket Index=$Particles.Initial.Spawngroup
    Inherit Velocity=false
    MaskSpawnGroup=false
    Mesh Sampling Type=::NewEnumerator5
    Sampled Position Offset Coordinate Space:ENiagaraMeshTransforms=::NewEnumerator3
    Skeletal Mesh=$User.SkeletalMesh
    Socket Sampling Mode=::NewEnumerator5
   )
   /SetVariables(Particles.FrontRibbonMask:NiagaraFloat=&OneMinusFloat(Float=&Round_Float(Float:NiagaraFloat=$System.NormalizedLoopAge))) {isSetParameters=true, module=SetVariables_#2}
  ]
  particleUpdateScript [
   &ParticleState()
   &Color(Color=&ColorFromCurve(ColorCurve=DataInterface() {propertyValues {
       AlphaCurve {keys [<at 0 0.12 RCIM_Cubic>, {arriveTangent=-0.4613, interpMode=RCIM_Cubic, leaveTangent=-0.4613, tangentMode=RCTM_User, time=1, value=0}]}
       BlueCurve {keys [<at 0 1.333 RCIM_Constant>,<at 0.1874 0.8766>,<at 0.95 0>,<at 1 0>]}
       GreenCurve {keys [<at 0 1.3003 RCIM_Constant>,<at 0.1874 0.6366>,<at 0.85 0>,<at 1 0>]}
       RedCurve {keys [<at 0 1.2707 RCIM_Constant>,<at 0.1874 0.4166>,<at 0.75 0>,<at 1 0>]}
      }}))
   &ScaleRibbonWidth(Ribbon Width Scale=&FloatFromCurve(FloatCurve=DataInterface() {propertyValues {Curve {keys [<at 0 1 RCIM_Cubic>, {arriveTangent=-0.033, interpMode=RCIM_Cubic, leaveTangent=-0.033, tangentMode=RCTM_User, time=1, value=2}]}}}))
  ]
  rendererClasses [&NiagaraRibbonRendererProperties]
  renderers [{rendererClass=&NiagaraRibbonRendererProperties}]
 }
 MainTrail {
  bEnabled=false
  emitterUpdateScript [
   &EmitterState(Life Cycle Mode=::NewEnumerator1, Loop Duration=0.025, UseLoopDelay=false)
   &SpawnBurst_Instantaneous(Spawn Count=4, Use Spawn Probability=false)
  ]
  particleSpawnScript [
   /SetVariables(Particles.Spawngroup:NiagaraInt32=&ModuloInt(A=&ReturnExecIndex(), B=4)) {isSetParameters=true, module=SetVariables_#1}
   &InitializeRibbon(
    Color=<rgba 1 1 1 0>
    Lifetime=0.185
    Module.Write Color=true
    Module.Write Lifetime=true
    Module.Write Mass=false
    Module.Write Material Random=true
    Module.Write Position=false
    Module.Write RibbonTwist=false
    Module.Write RibbonWidth=true
    Ribbon Width=10
   )
   &SkeletalMeshLocation(
    Bone / Socket Index=$Particles.Initial.Spawngroup
    Inherit Velocity=false
    MaskSpawnGroup=false
    Mesh Sampling Type=::NewEnumerator5
    Sampled Position Offset Coordinate Space:ENiagaraMeshTransforms=::NewEnumerator3
    Skeletal Mesh=$User.SkeletalMesh
    Socket Sampling Mode=::NewEnumerator5
   )
  ]
  particleUpdateScript [
   &ParticleState()
   &ScaleRibbonWidth(Ribbon Width Scale=&FloatFromCurve(FloatCurve=DataInterface() {propertyValues {Curve {keys [<at 0 1 RCIM_Cubic>, {arriveTangent=3.2031, interpMode=RCIM_Cubic, leaveTangent=3.1636, tangentMode=RCTM_User, time=1, value=2}]}}}))
   &DynamicMaterialParameters(
    Index 1 Param 1=&Add_Float(
     A=1
     B=&FloatFromCurve(FloatCurve=DataInterface() {propertyValues {Curve {keys [<at 0 0.25 RCIM_Cubic>,<at 1 0>]}}})
    )
    Module.Param0WriteEnabled1=true
    Module.Param1WriteEnabled1=true
    Module.Param2WriteEnabled1=true
    Module.Param3WriteEnabled1=true
    Write Parameter Index 0=false
    Write Parameter Index 1=true
   )
  ]
  rendererClasses [&NiagaraRibbonRendererProperties]
  renderers [{rendererClass=&NiagaraRibbonRendererProperties}]
 }
 MainTrail001 {
  emitterUpdateScript [
   &EmitterState(Life Cycle Mode=::NewEnumerator1, Loop Behavior=::NewEnumerator1, Loop Duration=0.3, UseLoopDelay=false)
   &SpawnBurst_Instantaneous(Spawn Count=4, Use Spawn Probability=false) {enabled=false}
   &SpawnRate(SpawnRate=120, Use Spawn Probability=false)
  ]
  particleSpawnScript [
   /SetVariables(Particles.Spawngroup:NiagaraInt32=&ModuloInt(A=&ReturnExecIndex(), B=4)) {isSetParameters=true, module=SetVariables_#1}
   &InitializeRibbon(
    Color=<rgba 1 1 1 0>
    Lifetime=0.185
    Module.Write Color=true
    Module.Write Lifetime=true
    Module.Write Mass=false
    Module.Write Material Random=true
    Module.Write Position=false
    Module.Write RibbonTwist=false
    Module.Write RibbonWidth=true
    Ribbon Width=10
   )
   &SkeletalMeshLocation(
    Bone / Socket Index=$Particles.Initial.Spawngroup
    Inherit Velocity=false
    Interpolate Mesh Sampling=false
    MaskSpawnGroup=false
    Mesh Sampling Type=::NewEnumerator5
    Sampled Position Offset Coordinate Space:ENiagaraMeshTransforms=::NewEnumerator3
    Skeletal Mesh=$User.SkeletalMesh
    Socket Sampling Mode=::NewEnumerator5
   )
  ]
  particleUpdateScript [
   &ParticleState()
   &ScaleRibbonWidth(Ribbon Width Scale=&FloatFromCurve(FloatCurve=DataInterface() {propertyValues {Curve {keys [<at 0 1 RCIM_Cubic>, {arriveTangent=3.2031, interpMode=RCIM_Cubic, leaveTangent=3.1636, tangentMode=RCTM_User, time=1, value=2}]}}}))
   &DynamicMaterialParameters(
    Index 1 Param 1=&Add_Float(
     A=1
     B=&FloatFromCurve(FloatCurve=DataInterface() {propertyValues {Curve {keys [<at 0 0.25 RCIM_Cubic>,<at 1 0>]}}})
    )
    Module.Param0WriteEnabled1=true
    Module.Param1WriteEnabled1=true
    Module.Param2WriteEnabled1=true
    Module.Param3WriteEnabled1=true
    Write Parameter Index 0=false
    Write Parameter Index 1=true
   )
  ]
  rendererClasses [&NiagaraRibbonRendererProperties]
  renderers [{rendererClass=&NiagaraRibbonRendererProperties}]
 }
}
summary {
 systemName=NS_Pickaxe_Jupiter_Trail
 emitters [
  {emitterName=LightTrail, rendererClasses [&NiagaraRibbonRendererProperties]}
  {emitterName=MainTrail001, rendererClasses [&NiagaraRibbonRendererProperties]}
  {bEnabled=false, emitterName=LightTrail001, rendererClasses [&NiagaraRibbonRendererProperties]}
  {bEnabled=false, emitterName=MainTrail, rendererClasses [&NiagaraRibbonRendererProperties]}
 ]
 userVariables [
  User.bIsFrontEndPreview {type=NiagaraBool}
  User.FEVec {type=Vector3f, description="", defaultValue {struct=Vector3f, value=<0 0 1000>}}
  User.PlayerSpeed {type=NiagaraFloat, defaultValue=0, description=""}
  User.RibbonIDOffset {type=NiagaraInt32}
  User.SkeletalMesh {
   type=NiagaraDataInterfaceSkeletalMesh
   defaultValue {
    struct=NiagaraExt_VariableValue_DataInterface
    value {dataInterface=&self:NiagaraDataInterfaceSkeletalMesh_0}
   }
  }
  User.SpawnRate {type=NiagaraFloat, defaultValue=360, description=""}
 ]
}
